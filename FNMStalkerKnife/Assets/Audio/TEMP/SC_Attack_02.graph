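# SC_Attack_02 is a temporary attack Sound Cue in the FNMStalkerKnife content, set up as a melee weapon sound.
# A Random node at the root picks among three equally weighted inputs, all of which are empty, so the cue references no sound waves here.
# It uses Fortnite's close-range stereo melee attack attenuation and weapon melee sound class, with a raised volume multiplier.
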SoundCue /FNMStalkerKnife/Assets/Audio/TEMP/SC_Attack_02
root=Random_0
nodes [{id=Random_0, children [null,null,null], properties {weights [1,1,1]}}]
properties {
 attenuationSettings=/Game/Sounds/Attenuation/Fort_Weapon_Melee_Attack_Close_Range_Stereo_Att
 maxDistance=10058.75
 soundClassObject=&Fort_WeaponMelee
 volumeMultiplier=2
}
